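# Translucent lit material for the Cyber Scavenger VinderTech umbrella cosmetic, with base color, normal and packed specular/metallic/roughness maps.
# Emissive combines a tinted glow mask and a rotating tiled cyber grid masked by FX2 red, faded by Fresnel, plus a sun/moon reflection highlight.
# Opacity follows Fresnel, blended toward Min_Opac by FX2 green; a refraction lerp is left unconnected.
# Attributes pass through MF_TechArtOverrides_accessories.
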Material /BRCosmetics/Gadgets/Assets/VinderTech_Umbrella/Materials/M_CyberScavenger_Umbrella
graph {
 properties {
  bUseMaterialAttributes=true
  bUsedWithMeshDeformer=true
  bUsedWithSkeletalMesh=true
  blendMode=BLEND_Translucent
  ditherOpacityMask=true
  pixelDepthOffsetMode=PDOM_Legacy
  translucencyLightingMode=TLM_SurfacePerPixelLighting
  usedShadingModels=Default Lit
 }
 nodes [
  n1 = TextureSample() {
   samplerType=SAMPLERTYPE_Normal
   texture=/BRCosmetics/Gadgets/Assets/VinderTech_Umbrella/Umbrella_Cyber/Textures/T_T_VinderTech_umbrella_CyberScavenger_MI_VinderTech_Umbrella_2_N
  }
  n2 = Fresnel(ExponentIn=ScalarParameter() {defaultValue=1, parameterName=Fresnel Exponent}, Normal=Transform(None=@n1.RGB))
  LinearInterpolate(A=ScalarParameter() {defaultValue=1, parameterName=RefractMin}, B=ScalarParameter() {defaultValue=0.25, parameterName=RefractMax}, Alpha=@n2)
  n3 = TextureSample() {texture=/BRCosmetics/Gadgets/Assets/VinderTech_Umbrella/Umbrella_Cyber/Textures/T_T_VinderTech_umbrella_CyberScavenger_MI_VinderTech_Umbrella_2_D}
  n4 = TextureSample() {
   samplerType=SAMPLERTYPE_Masks
   texture=/BRCosmetics/Gadgets/Assets/VinderTech_Umbrella/Umbrella_Cyber/Textures/T_T_VinderTech_umbrella_CyberScavenger_MI_VinderTech_Umbrella_2_S
  }
  n5 = TextureSample() {texture=/BRCosmetics/Gadgets/Assets/VinderTech_Umbrella/Umbrella_Cyber/Textures/T_T_VinderTech_umbrella_CyberScavenger_MI_VinderTech_Umbrella_2_E}
  n6 = TextureSample() {
   samplerType=SAMPLERTYPE_Masks
   texture=/BRCosmetics/Gadgets/Assets/VinderTech_Umbrella/Umbrella_Cyber/Textures/T_T_VinderTech_umbrella_CyberScavenger_MI_VinderTech_Umbrella_2_FX2
  }
  n7 = Multiply(A=TextureCoordinate(), B=ScalarParameter() {defaultValue=4, parameterName=GridTiling})
  n8 = VectorParameter() {parameterName=RotationCenter, defaultValue=<rgba -1.5 -1.5 0 1>}
  n9 = AppendVector(A=@n8.R, B=@n8.R)
  n10 = Time()
  n11 = ScalarParameter() {defaultValue=-1, parameterName=Speed_Rot}
  n12 = TextureSampleParameter2D(UVs=Rotator(Coordinate=Add(A=@n7, B=@n9), Time=Multiply(A=@n10, B=@n11))) {
   parameterName=Param
   samplerType=SAMPLERTYPE_LinearColor
   texture=/Game/Gadgets/Assets/VinderTech_Umbrella/Umbrella_Cyber/Textures/TilingCyberGrid_M
  }
  n13 = Multiply(A=@n6.R, B=@n12.G)
  n14 = LinearInterpolate(A=Multiply(A=VectorParameter() {parameterName=EmissiveTint, defaultValue=<rgba 0 5.624 15 1>}.RGB, B=Add(A=@n5.B, B=@n13)), Alpha=@n2) {constB=0}
  n15 = CollectionParameter() {
   collection=&FortniteMaterialParameters
   parameterName=SunAndMoonModelDirectionalVector
  }
  n16 = DotProduct(A=ComponentMask(None=@n15) {b=true, g=true, r=true}, B=ReflectionVectorWS())
  n17 = Multiply(
   A=Power(Base=Clamp(None=@n16), Exp=ScalarParameter() {defaultValue=90, parameterName=Spec Highlight Power}) {constExponent=20}
   B=ScalarParameter() {defaultValue=10, parameterName=Spec Intense}
  )
  n18 = LinearInterpolate(
   A=ScalarParameter() {defaultValue=0.1, parameterName=Fresnel Inner Transparency}
   B=ScalarParameter() {defaultValue=0.8, parameterName=Fresnel Outer Transparency}
   Alpha=@n2
  )
  n19 = ScalarParameter() {defaultValue=1, parameterName=Min_Opac}
  n20 = SetMaterialAttributes!(
   MaterialAttributes
   Base Color=NamedRerouteDeclaration(None=@n3.RGB) {name=basecolor, nodeColor=<rgba 0 0.0859 1 1>}
   Metallic=NamedRerouteDeclaration(None=@n4.G) {name=metallic, nodeColor=<rgba 1 0.2109 0 1>}
   Specular=NamedRerouteDeclaration(None=@n4.R) {name=specular, nodeColor=<rgba 0.0312 0 1 1>}
   Roughness=NamedRerouteDeclaration(None=@n4.B) {name=roughness, nodeColor=<rgba 1 0 0.1406 1>}
   Emissive Color=NamedRerouteDeclaration(None=Add(A=@n14, B=@n17)) {name=emissive, nodeColor=<rgba 0 0.4141 1 1>}
   Opacity=NamedRerouteDeclaration(None=LinearInterpolate(A=@n18, B=@n19, Alpha=@n6.G)) {name=opactiy, nodeColor=<rgba 0.9219 0 1 1>}
   Normal=NamedRerouteDeclaration(None=@n1.RGB) {name=normal, nodeColor=<rgba 1 0 0.4687 1>}
  ) {attributeSetTypes [
    #1
    #2
    #3
    #4
    #5
    #6
    #7
   ]}
  MaterialFunctionCall(In=@n20) {materialFunction=&MF_TechArtOverrides_accessories}
 ]
}
